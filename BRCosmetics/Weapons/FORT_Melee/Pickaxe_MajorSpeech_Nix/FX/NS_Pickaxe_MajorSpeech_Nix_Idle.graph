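# Niagara idle effect for the MajorSpeech Nix pickaxe: icy smoke sprites spawned along a thin cylinder, with spawning stopped while User.bIsSwinging is true.
# One emitter merges two spawn groups, soft drifting smoke and sparser, short-lived small sprites with randomized alpha.
# MergeHelper switches indexed by SpawnGroup, copied into VisibilityTag, select each group's parameters.
# Time-of-day inputs scale both groups' alpha and tint the smoke.
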NiagaraSystem /BRCosmetics/Weapons/FORT_Melee/Pickaxe_MajorSpeech_Nix/FX/NS_Pickaxe_MajorSpeech_Nix_Idle
dependencies {
 usedDataInterfaces [
  &NiagaraDataInterfaceColorCurve
  &NiagaraDataInterfaceCurve
  &NiagaraDataInterfaceSpriteRendererInfo
 ]
 usedDynamicInputs []
 usedModules [&SystemState]
}
emitters {IcySmoke {
  emitterUpdateScript [
   &EmitterState(Life Cycle Mode=::NewEnumerator1, Loop Behavior=::NewEnumerator1, Loop Duration Mode=::NewEnumerator1, UseLoopDelay=false)
   &SpawnRate(
    SpawnGroup=1
    SpawnRate=&MakeCustomFloatFromBool(Bool=$User.bIsSwinging, True Float=0, False Float=1)
    Use Spawn Probability=false
   )
   &SpawnRate(
    SpawnRate=&MakeCustomFloatFromBool(Bool=$User.bIsSwinging, True Float=0, False Float=10)
    Use Spawn Probability=false
   ) {module=SpawnRate001}
  ]
  particleSpawnScript [
   &InitializeParticle(
    Color=&MergeHelper_ColorSwitch(
     0=&RandomRangeLinearColor(Minimum=<rgba 0.3417 0.4071 0.4271 0.6>, Maximum=<rgba 0.74 0.8854 1 0.8>)
     1=&RandomRangeLinearColor(Minimum=<rgba 0.6 0.9067 1 1>, Maximum=<rgba 0.6 0.8067 1 1>)
     2=&RandomRangeLinearColor(Minimum=<rgba 0.6 0.9067 1 1>, Maximum=<rgba 0.6 0.8067 1 1>)
     3=?
     Index=$Emitter.SpawnGroup
     Num=::NewEnumerator0
    )
    Lifetime=&MergeHelper_FloatSwitch(
     0=&RandomRangeFloat(Minimum=1.4, Maximum=2)
     1=&RandomRangeFloat(Minimum=0.18, Maximum=0.26)
     2=?
     3=?
     Index=$Emitter.SpawnGroup
     Num=::NewEnumerator0
    )
    Sprite Rotation Mode=::NewEnumerator2
    Sprite Size Mode=::NewEnumerator3
    Sprite UV Mode=::NewEnumerator3
    Uniform Sprite Size=&MergeHelper_FloatSwitch(
     0=&RandomRangeFloat(Minimum=28, Maximum=35)
     1=&RandomRangeFloat(Minimum=9, Maximum=12)
     2=&RandomRangeFloat(Minimum=25, Maximum=25)
     3=?
     Index=$Emitter.SpawnGroup
     Num=::NewEnumerator0
    )
    UsePositionOffset=false
    Write Material Random=true
   )
   &ShapeLocation(
    Cylinder Height=70
    Cylinder Height Midpoint=0
    Cylinder Radius=1
    MaskSpawnGroup=false
    Offset=<0 0 8>
    Offset Coordinate Space=::Local
    Offset Mode=::NewEnumerator0
    Override Seed=false
    Shape Primitive=::NewEnumerator1
    Surface Only=false
    UseOwnerScale=true
   )
   &CameraOffset(Camera Offset Amount=15)
   &AddVelocity(Velocity=&MergeHelper_VectorSwitch(
     0=&RandomRangeVector(Minimum=<0 0 0>, Maximum=<10 0 30>)
     1=<0 0 0>
     2=?
     3=?
     Index=$Emitter.SpawnGroup
     Num=::NewEnumerator0
    ))
   /SetVariables(Particles.VisibilityTag=$Emitter.SpawnGroup) {isSetParameters=true}
   /SetVariables(Particles.DynamicParametersAlt:Vector4f=<1 1 1 1>) {isSetParameters=true}
  ]
  particleUpdateScript [
   &ParticleState()
   &Drag(Use Linear Drag=true, Use Rotational Drag=true)
   &ScaleColor(
    Scale Alpha=&MergeHelper_FloatSwitch(
     0=&FloatFromCurve(
      FloatCurve=DataInterface() {propertyValues {Curve {keys [
          {interpMode=RCIM_Cubic, tangentMode=RCTM_User, time=0, value=0}
          {interpMode=RCIM_Cubic, tangentMode=RCTM_User, time=0.4, value=1}
          {tangentMode=RCTM_User, time=1, value=0}
         ]}}}
      Scale Curve=&Multiply_Float(
       A=0.085
       B=&FNTimeOfDayScalar(Night Scale=0.9, Morning Scale=0.8, Day Scale=0.9, Evening Scale=0.85)
      )
     )
     1=&Multiply_Float(
      A=&UniformAOrBFloat(A=0, Evaluation Type=::NewEnumerator1, Recalculate Random Each Loop=true, Override Seed=false)
      B=&FNTimeOfDayScalar(Night Scale=0.9, Morning Scale=0.8, Day Scale=0.9, Evening Scale=0.85)
     )
     2=?
     3=?
     Index=$Particles.VisibilityTag
     Num=::NewEnumerator0
    )
    Scale RGB=&MergeHelper_VectorSwitch(
     0=&MakeVectorFromLinearColorRGB(Color=&ColorFromCurve(
       ColorCurve=DataInterface() {propertyValues {
         AlphaCurve {keys [<at 0 1 RCIM_Cubic>]}
         BlueCurve {keys [<at 0.0932 1 RCIM_Cubic>,<at 0.125 1>,<at 0.5 1>,<at 0.875 1 RCIM_Cubic>,<at 0.905 1>]}
         GreenCurve {keys [
           <at 0.0932 0.5833 RCIM_Cubic>
           <at 0.125 1>
           <at 0.5 1>
           {arriveTangent=-1.0288, interpMode=RCIM_Cubic, leaveTangent=-1.0288, time=0.875, value=1}
           <at 0.905 0.5833>
          ]}
         RedCurve {keys [
           <at 0.0932 0.5 RCIM_Cubic>
           <at 0.125 1>
           <at 0.5 1>
           {arriveTangent=-1.2346, interpMode=RCIM_Cubic, leaveTangent=-1.2346, time=0.875, value=1}
           <at 0.905 0.5>
          ]}
        }}
       CurveIndex=Linked() {linkedVariable {name=NPC.FortniteNPC.RemappedTimeOfDay_0-1, type=NiagaraFloat}}
      ))
     1=<1 1 1>
     2=?
     3=?
     Index=$Particles.VisibilityTag
     Num=::NewEnumerator0
    )
    ScaleA=true
    ScaleRGB=true
   )
   &ScaleSpriteSize(Scale Factor=&MergeHelper_Vector2Switch(
     0=&Vector2DFromFloat(Value=&FloatFromCurve(FloatCurve=DataInterface() {propertyValues {Curve {keys [{arriveTangent=0.3685, time=0, value=0.8}, <at 0.3 0.9>, {arriveTangent=0.1633, leaveTangent=0.1633, time=1, value=1}]}}}))
     1=&Vector2DFromFloat(Value=&FloatFromCurve(FloatCurve=DataInterface() {propertyValues {Curve {keys [{arriveTangent=0.3685, time=0, value=0.8}, <at 0.3 0.9>, {arriveTangent=0.1633, leaveTangent=0.1633, time=1, value=1}]}}}))
     2=?
     3=?
     4=?
     Index=$Particles.VisibilityTag
     Num=::NewEnumerator0
    ))
   &SpriteRotationRate()
   &SubUVAnimation(
    "Sprite Renderer "=DataInterface() {propertyValues {SpriteRenderer=&self:IcySmoke_0.NiagaraSpriteRendererProperties_1}}
    UseEndFrame=false
    UseStartFrame=false
   )
   &DynamicMaterialParameters(
    Index 0 Param 1=1
    Index 0 Param 4=1
    Module.Index0Float0Write=true
    Module.Index0Float1Write=true
    Module.Index0Float2Write=true
    Module.Index0Float3Write=true
   )
   &SolveForcesAndVelocity(Clamp Velocity=false, Limit Acceleration=false)
  ]
  rendererClasses [&NiagaraSpriteRendererProperties]
  renderers [
   {rendererClass=&NiagaraSpriteRendererProperties}
   {rendererClass=&NiagaraSpriteRendererProperties, rendererIndex=1}
  ]
 }}
summary {
 systemName=NS_Pickaxe_MajorSpeech_Nix_Idle
 emitters [{emitterName=IcySmoke, rendererClasses [&NiagaraSpriteRendererProperties]}]
 userVariables [
  User.bIsSwinging {type=NiagaraBool}
  User.RibbonIDOffset {type=NiagaraInt32}
  User.SwingSectionIndex {type=NiagaraInt32}
 ]
}
