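# Anim Blueprint for the MajorSpeech_Nix pickaxe that adds springy secondary motion to its ropes.
# The cached input pose feeds two Anim Dynamics chains, blended by bIsCrouching.
# The crouch chain first yaws base_rope_a; the standing chain uses different tuning, simulates dyn_rope_b_01 twice and rotation-constrains dyn_rope_a_02 to dyn_rope_b_03.
# Update logic sets opposing front-end alpha variables, which the AnimGraph never reads.
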AnimBlueprint /BRCosmetics/Weapons/FORT_Melee/Pickaxe_MajorSpeech_Nix/Meshes/MajorSpeech_Nix_Axe_AnimBP
target_skeleton=&/MajorSpeech_Nix_Axe_Skeleton
graphs [
 AnimGraph {nodes [
   n1 = "Transform (Modify) Bone - Bone: base_rope_a"!(
    ComponentPose=Local To Component(LocalPose=Use cached pose 'StarCache'())
    bAlphaBoolEnabled:bool=True
    Translation="0,0,0"
    Rotation="0,50,0"
    Scale="1,1,1"
   ) -> (Pose) {
    alphaInputType=Bool
    rotationMode=BMM_Additive
    rotationSpace=BCS_BoneSpace
    translationMode=BMM_Additive
    translationSpace=BCS_BoneSpace
    alphaBoolBlend {blendInTime=0.2, blendOutTime=0.2}
    boneToModify {boneName=base_rope_a}
    rotation {yaw=50}
   }
   n2 = Anim Dynamics - Bone: dyn_rope_a_01(ComponentPose=@n1, Alpha=1, ExternalForce="0,0,0") {
    angularDampingOverride=0.999
    angularSpringConstant=100
    bAngularSpring=true
    bOverrideAngularDamping=true
    bUseGravityOverride=true
    gravityScale=0.35
    simulationSpace=RootRelative
    boundBone {boneName=dyn_rope_a_01}
    componentAppliedLinearAccClamp=<3000 3000 3000>
    componentLinearAccScale=<1 1 1>
    componentLinearVelScale=<1 1 1>
    gravityOverride {z=-980}
    physicsBodyDefinitions [{
      boundBone {boneName=dyn_rope_a_01}
      boxExtents=<22 22 22>
      constraintSetup {angularLimitsMax=<30 90 30>, angularLimitsMin=<-30 -20 -15>, angularTarget=<1 0 0>}
      localJointOffset=<4 0 0>
     }]
    relativeSpaceBone {boneName=b_AR_Root}
    simSpaceSettings {maxAngularAcceleration=3000, maxAngularVelocity=3000, simSpaceAngularAlpha=0.6}
   }
   n3 = Anim Dynamics - Bone: dyn_rope_b_01(ComponentPose=@n2, Alpha=1, ExternalForce="0,0,0") {
    angularDampingOverride=0.999
    angularSpringConstant=100
    bAngularSpring=true
    bOverrideAngularDamping=true
    bUseGravityOverride=true
    gravityScale=0.35
    simulationSpace=RootRelative
    boundBone {boneName=dyn_rope_b_01}
    componentAppliedLinearAccClamp=<3000 3000 3000>
    componentLinearAccScale=<1 1 1>
    componentLinearVelScale=<1 1 1>
    gravityOverride {z=-980}
    physicsBodyDefinitions [{
      boundBone {boneName=dyn_rope_b_01}
      boxExtents=<22 22 22>
      constraintSetup {angularLimitsMax=<30 80 10>, angularLimitsMin=<-30 -25 -10>, angularTarget=<1 0 0>}
      localJointOffset=<4.5 0 0>
     }]
    simSpaceSettings {maxAngularAcceleration=3000, maxAngularVelocity=3000, simSpaceAngularAlpha=0.6}
   }
   n4 = Anim Dynamics - Bone: dyn_rope_b_02(ComponentPose=@n3, Alpha=1, ExternalForce="0,0,0") {
    angularDampingOverride=0.999
    angularSpringConstant=500
    bAngularSpring=true
    bOverrideAngularDamping=true
    bUseGravityOverride=true
    gravityScale=0.35
    simulationSpace=RootRelative
    boundBone {boneName=dyn_rope_b_02}
    componentAppliedLinearAccClamp=<3000 3000 3000>
    componentLinearAccScale=<1 1 1>
    componentLinearVelScale=<1 1 1>
    gravityOverride {z=-980}
    physicsBodyDefinitions [{
      boundBone {boneName=dyn_rope_b_02}
      boxExtents=<22 22 22>
      constraintSetup {angularLimitsMax=<30 30 20>, angularLimitsMin=<-30 -25 -15>, angularTarget=<1 0 0>}
      localJointOffset=<4.5 0 0>
     }]
    simSpaceSettings {maxAngularAcceleration=3000, maxAngularVelocity=3000, simSpaceAngularAlpha=0.6}
   }
   n5 = Anim Dynamics - Bone: dyn_rope_b_03(ComponentPose=@n4, Alpha=1, ExternalForce="0,0,0") {
    angularDampingOverride=0.999
    angularSpringConstant=500
    bAngularSpring=true
    bOverrideAngularDamping=true
    bUseGravityOverride=true
    gravityScale=0.35
    simulationSpace=RootRelative
    boundBone {boneName=dyn_rope_b_03}
    componentAppliedLinearAccClamp=<3000 3000 3000>
    componentLinearAccScale=<1 1 1>
    componentLinearVelScale=<1 1 1>
    gravityOverride {z=-980}
    physicsBodyDefinitions [{
      boundBone {boneName=dyn_rope_b_03}
      boxExtents=<22 22 22>
      constraintSetup {angularLimitsMax=<10 15 20>, angularLimitsMin=<-10 -25 -15>, angularTarget=<1 0 0>}
      localJointOffset=<4.5 0 0>
     }]
    simSpaceSettings {maxAngularAcceleration=3000, maxAngularVelocity=3000, simSpaceAngularAlpha=0.6}
   }
   n6 = Anim Dynamics - Bone: dyn_rope_b_04(ComponentPose=@n5, Alpha=1, ExternalForce="0,0,0") {
    angularDampingOverride=0.999
    angularSpringConstant=500
    bAngularSpring=true
    bOverrideAngularDamping=true
    bUseGravityOverride=true
    gravityScale=0.35
    simulationSpace=RootRelative
    boundBone {boneName=dyn_rope_b_04}
    componentAppliedLinearAccClamp=<3000 3000 3000>
    componentLinearAccScale=<1 1 1>
    componentLinearVelScale=<1 1 1>
    gravityOverride {z=-980}
    physicsBodyDefinitions [{
      boundBone {boneName=dyn_rope_b_04}
      boxExtents=<22 22 22>
      constraintSetup {angularLimitsMax=<10 15 20>, angularLimitsMin=<-10 -25 -15>, angularTarget=<1 0 0>}
      localJointOffset=<4.5 0 0>
     }]
    simSpaceSettings {maxAngularAcceleration=3000, maxAngularVelocity=3000, simSpaceAngularAlpha=0.6}
   }
   n7 = Anim Dynamics - Bone: dyn_rope_a_01(ComponentPose=Local To Component(LocalPose=Use cached pose 'StarCache'()), Alpha=1, ExternalForce="0,0,0") {
    angularDampingOverride=0.999
    angularSpringConstant=100
    bAngularSpring=true
    bOverrideAngularDamping=true
    bUseGravityOverride=true
    gravityScale=0.35
    simulationSpace=RootRelative
    boundBone {boneName=dyn_rope_a_01}
    componentAppliedLinearAccClamp=<3000 3000 3000>
    componentLinearAccScale=<0.7 0.7 0.7>
    componentLinearVelScale=<1 1 1>
    gravityOverride {z=-980}
    physicsBodyDefinitions [{
      boundBone {boneName=dyn_rope_a_01}
      boxExtents=<25 25 25>
      constraintSetup {angularLimitsMax=<30 90 60>, angularLimitsMin=<-30 -20 -30>, angularTarget=<1 0 0>}
      localJointOffset=<4 0 0>
     }]
    relativeSpaceBone {boneName=b_AR_Root}
    simSpaceSettings {maxAngularAcceleration=3000, maxAngularVelocity=3000, simSpaceAngularAlpha=0.8}
   }
   n8 = Anim Dynamics - Bone: dyn_rope_b_01(ComponentPose=@n7, Alpha=1, ExternalForce="0,0,0") {
    angularDampingOverride=0.999
    angularSpringConstant=100
    bAngularSpring=true
    bOverrideAngularDamping=true
    bUseGravityOverride=true
    gravityScale=0.35
    simulationSpace=RootRelative
    alphaScaleBiasClamp {bInterpResult=true}
    boundBone {boneName=dyn_rope_b_01}
    componentAppliedLinearAccClamp=<3000 3000 3000>
    componentLinearAccScale=<0.7 0.7 0.7>
    componentLinearVelScale=<1 1 1>
    gravityOverride {z=-980}
    physicsBodyDefinitions [{
      boundBone {boneName=dyn_rope_b_01}
      boxExtents=<25 25 25>
      constraintSetup {angularLimitsMax=<30 80 10>, angularLimitsMin=<-30 -10 -20>, angularTarget=<1 0 0>}
      localJointOffset=<4.5 0 0>
     }]
    simSpaceSettings {maxAngularAcceleration=3000, maxAngularVelocity=3000, simSpaceAngularAlpha=0.8}
   }
   n9 = Anim Dynamics - Bone: dyn_rope_b_01(ComponentPose=@n8, Alpha=1, ExternalForce="0,0,0") {
    angularDampingOverride=0.999
    angularSpringConstant=100
    bAngularSpring=true
    bOverrideAngularDamping=true
    bUseGravityOverride=true
    gravityScale=0.35
    simulationSpace=RootRelative
    alphaScaleBiasClamp {bInterpResult=true}
    boundBone {boneName=dyn_rope_b_01}
    componentAppliedLinearAccClamp=<3000 3000 3000>
    componentLinearAccScale=<0.7 0.7 0.7>
    componentLinearVelScale=<1 1 1>
    gravityOverride {z=-980}
    physicsBodyDefinitions [{
      boundBone {boneName=dyn_rope_b_01}
      boxExtents=<25 25 25>
      constraintSetup {angularLimitsMax=<30 80 0>, angularLimitsMin=<-30 -10 -20>, angularTarget=<1 0 0>}
      localJointOffset=<4.5 0 0>
     }]
    simSpaceSettings {maxAngularAcceleration=3000, maxAngularVelocity=3000, simSpaceAngularAlpha=0.8}
   }
   n10 = Anim Dynamics - Bone: dyn_rope_b_02(ComponentPose=@n9, Alpha=1, ExternalForce="0,0,0") {
    angularDampingOverride=0.999
    angularSpringConstant=500
    bAngularSpring=true
    bOverrideAngularDamping=true
    bUseGravityOverride=true
    gravityScale=0.35
    simulationSpace=RootRelative
    alphaScaleBiasClamp {bInterpResult=true}
    boundBone {boneName=dyn_rope_b_02}
    componentAppliedLinearAccClamp=<3000 3000 3000>
    componentLinearAccScale=<0.7 0.7 0.7>
    componentLinearVelScale=<1 1 1>
    gravityOverride {z=-980}
    physicsBodyDefinitions [{
      boundBone {boneName=dyn_rope_b_02}
      boxExtents=<25 25 25>
      constraintSetup {angularLimitsMax=<30 30 20>, angularLimitsMin=<-30 -25 -45>, angularTarget=<1 0 0>}
      localJointOffset=<4.5 0 0>
     }]
    simSpaceSettings {maxAngularAcceleration=3000, maxAngularVelocity=3000, simSpaceAngularAlpha=0.8}
   }
   n11 = Anim Dynamics - Bone: dyn_rope_b_03(ComponentPose=@n10, Alpha=1, ExternalForce="0,0,0") {
    angularDampingOverride=0.999
    angularSpringConstant=500
    bAngularSpring=true
    bOverrideAngularDamping=true
    bUseGravityOverride=true
    gravityScale=0.35
    simulationSpace=RootRelative
    boundBone {boneName=dyn_rope_b_03}
    componentAppliedLinearAccClamp=<3000 3000 3000>
    componentLinearAccScale=<0.7 0.7 0.7>
    componentLinearVelScale=<1 1 1>
    gravityOverride {z=-980}
    physicsBodyDefinitions [{
      boundBone {boneName=dyn_rope_b_03}
      boxExtents=<25 25 25>
      constraintSetup {angularLimitsMax=<10 10 20>, angularLimitsMin=<-10 -25 -15>, angularTarget=<1 0 0>}
      localJointOffset=<4.5 0 0>
     }]
    simSpaceSettings {maxAngularAcceleration=3000, maxAngularVelocity=3000, simSpaceAngularAlpha=0.8}
   }
   n12 = Anim Dynamics - Bone: dyn_rope_b_04(ComponentPose=@n11, Alpha=1, ExternalForce="0,0,0") {
    angularDampingOverride=0.999
    angularSpringConstant=500
    bAngularSpring=true
    bOverrideAngularDamping=true
    bUseGravityOverride=true
    gravityScale=0.35
    simulationSpace=RootRelative
    boundBone {boneName=dyn_rope_b_04}
    componentAppliedLinearAccClamp=<3000 3000 3000>
    componentLinearAccScale=<0.7 0.7 0.7>
    componentLinearVelScale=<1 1 1>
    gravityOverride {z=-980}
    physicsBodyDefinitions [{
      boundBone {boneName=dyn_rope_b_04}
      boxExtents=<25 25 25>
      constraintSetup {angularLimitsMax=<10 15 20>, angularLimitsMin=<-10 -25 0>, angularTarget=<1 0 0>}
      localJointOffset=<4.5 0 0>
     }]
    simSpaceSettings {maxAngularAcceleration=3000, maxAngularVelocity=3000, simSpaceAngularAlpha=0.8}
   }
   n13 = "Constraint - dyn_rope_a_02 (R)"(ComponentPose=@n12, Alpha=1, ConstraintWeights_0=1) {
    boneToModify {boneName=dyn_rope_a_02}
    constraintSetup [{transformType=Rotation, targetBone {boneName=dyn_rope_b_03}}]
    constraintWeights [1]
   }
   Output Pose(Result=Blend Poses by bool(BlendPose_0=Component To Local(ComponentPose=@n6), BlendPose_1=Component To Local(ComponentPose=@n13), bActiveValue=Get bIsCrouching())) {layerGroup=None}
   AnimGraphNode_SubInput_0 = Input Pose()
   Save cached pose 'StarCache'(Pose=@AnimGraphNode_SubInput_0) {cacheName=StarCache}
  ]}
 EventGraph {nodes [
   Event BlueprintUpdateAnimation()
   > n1 = Branch(Condition=Get bIsInFrontEnd())
   > Set alpha_FE(alpha_FE:double=1) -> (Output_Get:double)
   > Set alpha_NoFE(alpha_NoFE:double=0) -> (Output_Get:double)
   Set alpha_FE(execute=@n1.else, alpha_FE=0)
   > Set alpha_NoFE(alpha_NoFE=1)
  ]}
]
variables [
 alpha_FE {type=number}
 alpha_NoFE {type=number}
]
